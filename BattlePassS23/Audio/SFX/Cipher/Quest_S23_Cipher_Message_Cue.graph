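# Sound cue for a Battle Pass Season 23 Cipher quest message, playing as an attenuated, looping sound in the quests sound class.
# It layers two looping waves, a message loop and a comm-tuning loop, each through its own enveloper with a short volume fade-in.
# A mixer combines both layers, weighting the tuning loop far louder than the message loop.
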SoundCue /BattlePassS23/Audio/SFX/Cipher/Quest_S23_Cipher_Message_Cue
root=Mixer_0
nodes [
 {
  id=Enveloper_0
  children [WavePlayer_0]
  properties {pitchCurve {editorCurveData {keys [{time=0}]}}, volumeCurve {editorCurveData {keys [<at 0 0>,<at 0.1 1>]}}}
 }
 {
  id=Enveloper_1
  children [WavePlayer_13]
  properties {pitchCurve {editorCurveData {keys [{time=0}]}}, volumeCurve {editorCurveData {keys [<at 0 0>,<at 0.1 1>]}}}
 }
 {id=Mixer_0, children [Enveloper_0,Enveloper_1], properties {inputVolume [0.15,1.5]}}
 {id=WavePlayer_0, properties {bLooping=true, soundWaveAssetPtr=&/Quest_S23_Cipher_Message_LP}}
 {id=WavePlayer_13, properties {bLooping=true, soundWaveAssetPtr=&/Quest_S23_Cipher_Comm_Tuning_LP}}
]
properties {
 attenuationSettings=&/Quest_S23_Cipher_Message_Atten
 maxDistance=1300
 soundClassObject=&Fort_Quests
 volumeMultiplier=1
}
